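# Niagara swing trail effect for the Captain Shine guitar pickaxe cosmetic, built from two short-lived ribbon emitters, TrailSharp and TrailSoft.
# Both use shared AnimTrail functions to spawn, initialize and update ribbons from bone or socket pairs, with differing pairs and dynamic material parameters.
# Color scales by a system time-of-day tint.
# Alpha multiplies a system fade by an over-life curve, itself scaled by a curve indexed on Fortnite time of day.
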NiagaraSystem /SparksCosmetics/Instrument/Guitar/CaptainShine/Effects/NS_Pickaxe_Guitar_CaptainShine_SwingTrail
dependencies {
 usedDataInterfaces [&NiagaraDataInterfaceCurve, &NiagaraDataInterfaceSocketReader, &NiagaraDataInterfaceColorCurve]
 usedDynamicInputs [&ColorFromCurve, &MergeHelper_FloatSwitch]
 usedModules [
  &self:SystemSpawnScript.SystemScriptSource.SystemScriptGraph.NiagaraNodeAssignment_4.SetVariables_#1
  &self:SystemSpawnScript.SystemScriptSource.SystemScriptGraph.NiagaraNodeAssignment_5.SetVariables_#2
  &SystemState
  &self:SystemSpawnScript.SystemScriptSource.SystemScriptGraph.NiagaraNodeAssignment_0.SetVariables_#3
  &self:SystemSpawnScript.SystemScriptSource.SystemScriptGraph.NiagaraNodeAssignment_1.SetVariables_#4
  &self:SystemSpawnScript.SystemScriptSource.SystemScriptGraph.NiagaraNodeAssignment_2.SetVariables_#5
 ]
}
emitters {
 TrailSharp {
  emitterSpawnScript [&FN_RibbonIDOffsetUtility()]
  emitterUpdateScript [
   &EmitterState()
   &FN_SpawnAnimTrail()
  ]
  particleSpawnScript [
   &FN_InitializeAnimTrailBoneIndecies(First Bone or Socket Index=1, Second Bone or Socket Index=2)
   &FN_InitializeAnimTrail(Color=<rgba 1 1 1 0.8>, Lifetime=0.16)
  ]
  particleUpdateScript [
   &FN_UpdateAnimTrail() {module=FN_UpdateAnimTrail001}
   &SolveForcesAndVelocity(Clamp Velocity=false, Limit Acceleration=false)
   &ScaleColor(
    Scale Alpha=&Multiply_Float(
     A=$System.AlphaFade
     B=&FloatFromCurve(
      FloatCurve=DataInterface() {propertyValues {Curve {keys [
          {interpMode=RCIM_Cubic, leaveTangent=-2.2741, leaveTangentWeight=0.5724, tangentMode=RCTM_Break, time=0, value=1}
          <at 0.4 0.15>
          {arriveTangent=-0.2139, interpMode=RCIM_Cubic, tangentMode=RCTM_Break, time=1, value=0}
         ]}}}
      Scale Curve=&FloatFromCurve(
       FloatCurve=DataInterface() {propertyValues {Curve {keys [
           <at 0 0.9 RCIM_Cubic>
           {arriveTangent=0.2255, interpMode=RCIM_Cubic, leaveTangent=0.2255, time=0.2025, value=0.95}
           {arriveTangentWeight=0.0858, interpMode=RCIM_Cubic, leaveTangentWeight=0.0558, tangentMode=RCTM_User, time=0.4434, value=1}
           {arriveTangentWeight=0.0572, interpMode=RCIM_Cubic, leaveTangentWeight=0.1462, tangentMode=RCTM_User, time=0.6, value=1}
           <at 1 0.9 RCIM_Cubic>
          ]}}}
       CurveIndex=Linked() {linkedVariable {name=NPC.FortniteNPC.TimeOfDay_0-1, type=NiagaraFloat}}
      )
     )
    )
    Scale RGB=&MakeVectorFromLinearColorRGB(Color=$System.ToDColor)
    ScaleA=true
    ScaleRGB=true
   )
   &DynamicMaterialParameters(
    Index 0 Param 1=0.2
    Index 0 Param 2=0.2
    Index 1 Param 4=1
    Index 2 Param 1=15
    Index 2 Param 2=4
    Module.Index0Float0Write=true
    Module.Index0Float1Write=true
    Module.Index0Float2Write=true
    Module.Index0Float3Write=true
    Module.Index1Float0Write=true
    Module.Index1Float1Write=true
    Module.Index1Float2Write=true
    Module.Index1Float3Write=true
    Module.Index2Float0Write=true
    Module.Index2Float1Write=true
    Module.Index2Float2Write=false
    Module.Index2Float3Write=false
    Write Parameter Index 1=::NewEnumerator1
    Write Parameter Index 2=::NewEnumerator1
   )
  ]
  rendererClasses [&NiagaraRibbonRendererProperties]
  renderers [{rendererClass=&NiagaraRibbonRendererProperties}]
 }
 TrailSoft {
  emitterSpawnScript [&FN_RibbonIDOffsetUtility()]
  emitterUpdateScript [
   &EmitterState()
   &FN_SpawnAnimTrail()
  ]
  particleSpawnScript [
   &FN_InitializeAnimTrailBoneIndecies(Second Bone or Socket Index=2)
   &FN_InitializeAnimTrail(
    Color=&MakeCustomLinearColorFromBool(Bool=false, True Color=<rgba 0.4 0.4 0.4 1>, False Color=<rgba 1 1 1 1>)
    Lifetime=0.16
   )
  ]
  particleUpdateScript [
   &FN_UpdateAnimTrail() {module=FN_UpdateAnimTrail001}
   &ScaleColor(
    Scale Alpha=&Multiply_Float(
     A=$System.AlphaFade
     B=&FloatFromCurve(
      FloatCurve=DataInterface() {propertyValues {Curve {keys [
          {interpMode=RCIM_Cubic, leaveTangent=-2.2741, leaveTangentWeight=0.5724, tangentMode=RCTM_Break, time=0, value=1}
          <at 0.4 0.15>
          {arriveTangent=-0.2139, interpMode=RCIM_Cubic, tangentMode=RCTM_Break, time=1, value=0}
         ]}}}
      Scale Curve=&FloatFromCurve(
       FloatCurve=DataInterface() {propertyValues {Curve {keys [
           <at 0 0.9 RCIM_Cubic>
           {arriveTangent=0.2255, interpMode=RCIM_Cubic, leaveTangent=0.2255, time=0.2025, value=0.95}
           {arriveTangentWeight=0.0858, interpMode=RCIM_Cubic, leaveTangentWeight=0.0558, tangentMode=RCTM_User, time=0.4434, value=1}
           {arriveTangentWeight=0.0572, interpMode=RCIM_Cubic, leaveTangentWeight=0.1462, tangentMode=RCTM_User, time=0.6, value=1}
           <at 1 0.9 RCIM_Cubic>
          ]}}}
       CurveIndex=Linked() {linkedVariable {name=NPC.FortniteNPC.TimeOfDay_0-1, type=NiagaraFloat}}
      )
     )
    )
    Scale RGB=&MakeVectorFromLinearColorRGB(Color=$System.ToDColor)
    ScaleA=true
    ScaleRGB=true
   )
   &SolveForcesAndVelocity(Clamp Velocity=false, Limit Acceleration=false)
   &DynamicMaterialParameters(
    Index 0 Param 1=1
    Index 1 Param 4=1
    Index 2 Param 1=10
    Index 2 Param 2=1
    Index 2 Param 3=0.5
    Index 2 Param 4=0.6
    Module.Index0Float0Write=true
    Module.Index0Float1Write=true
    Module.Index0Float2Write=true
    Module.Index0Float3Write=true
    Module.Index1Float0Write=true
    Module.Index1Float1Write=true
    Module.Index1Float2Write=true
    Module.Index1Float3Write=true
    Module.Index2Float0Write=true
    Module.Index2Float1Write=true
    Module.Index2Float2Write=false
    Module.Index2Float3Write=false
    Write Parameter Index 1=::NewEnumerator1
    Write Parameter Index 2=::NewEnumerator1
   )
  ]
  rendererClasses [&NiagaraRibbonRendererProperties]
  renderers [{rendererClass=&NiagaraRibbonRendererProperties}]
 }
}
summary {
 systemName=NS_Pickaxe_Guitar_CaptainShine_SwingTrail
 emitters [
  {emitterName=TrailSharp, rendererClasses [&NiagaraRibbonRendererProperties]}
  {emitterName=TrailSoft, rendererClasses [&NiagaraRibbonRendererProperties]}
 ]
 userVariables [User.SwingSectionIndex {type=NiagaraInt32}]
}
